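; Passive gameplay ability for the temple key prototype that guides its player toward the nearest pressable brazier switch.
; It selects the closest eligible lock actor on activation, then searches again when that switch is interacted with.
; After a net sync, the cosmetic side registers a marker on the target.
; The server computes direction on a looping timer and replicates it for a Niagara trail, so non-relevant targets still point correctly.
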(event Custom|LookForAnotherDestination
  (bind _returnvalue (CallFunction|GetAvatarasFortPlayerPawn))
  (bind _outactors (Actor|GetAllActorsOfClass (Variables|Default|GetLockActorClass)))
  (for _array_element _outactors
    (bind _asbp_brazier_switch (Utilities|Casting|CastToBP_BrazierSwitch _array_element))
    (if (and (< (Transformation|GetDistanceTo _array_element _returnvalue) (Variables|Default|GetThisDistance)) (Variables|/Hera/Prototype/LevelDesign/Allen/BP_BrazierSwitch.BP_BrazierSwitch_C|GetButton%20Can%20Be%20Pressed _asbp_brazier_switch))
      (Variables|Default|SetThisDistance (Transformation|GetDistanceTo _array_element _returnvalue))
      (Variables|Default|SetThisActor _array_element)))
  (bind _thisactor (Variables|Default|GetThisActor)
    (:then
      (bind _output_get (Variables|Default|SetLockActor _thisactor))
      (bind _asbp_brazier_switch_2 (Utilities|Casting|CastToBP_BrazierSwitch _output_get))
      (Interaction|BindEventtoOnInteract _asbp_brazier_switch_2 (delegate Custom|DestinationUpdated)))
    (:else)))

(event Ability|EventActivateAbility
  (bind _returnvalue (CallFunction|GetAvatarasFortPlayerPawn))
  (CallFunction|InitAbility)
  (Utilities|FlowControl|Sequence
    (:then_0
      (bind _outactors (Actor|GetAllActorsOfClass (Variables|Default|GetLockActorClass)))
      (for _array_element _outactors
        (bind _asbp_brazier_switch (Utilities|Casting|CastToBP_BrazierSwitch _array_element))
        (if (and (< (Transformation|GetDistanceTo _array_element _returnvalue) (Variables|Default|GetThisDistance)) (Variables|/Hera/Prototype/LevelDesign/Allen/BP_BrazierSwitch.BP_BrazierSwitch_C|GetButton%20Can%20Be%20Pressed _asbp_brazier_switch))
          (Variables|Default|SetThisDistance (Transformation|GetDistanceTo _array_element _returnvalue))
          (Variables|Default|SetThisActor _array_element)))
      (bind _thisactor (Variables|Default|GetThisActor)
        (:then
          (bind _output_get (Variables|Default|SetLockActor _thisactor))
          (bind _asbp_brazier_switch_2 (Utilities|Casting|CastToBP_BrazierSwitch _output_get))
          (Interaction|BindEventtoOnInteract _asbp_brazier_switch_2 (delegate Custom|DestinationUpdated)))
        (:else)))
    (:then_1
      (Ability|Tasks|WaitNetSync "BothWait"
        (:OnSync
          (Networking|CanExecuteCosmeticEvents
            (:True
              (CallFunction|PingMarkTarget))
            (:False
              (CallFunction|DrawTrailtoVault)
              (Utilities|Time|SetTimerbyEvent (EventDispatchers|CreateEvent :!event Custom|DrawTrailToVault) (Variables|Default|GetTimeBetweenLines) true))))))))

(event Custom|DestinationUpdated
  (params (SelfActor :type Building%20Actor%20Object%20Reference) (InteractingPawn :type Fort%20Pawn%20Object%20Reference))
  (Variables|Default|SetThisActor 0)
  (Variables|Default|SetLockActor 0)
  (CallFunction|LookforAnotherDestination))

(event Custom|DrawTrailToVault
  ;[1 1 1] From the server do the math and pass the rotation to the client. This is done so that if the actor is not relevant to the client the server still points in the correct direction.
  (if (Networking|IsServer)
    (Variables|Default|GetLockActor
      (:then
        (CallFunction|ReplicatedDrawTrailtoVault (Math|Rotator|MakeRotfromX (* (- (Transformation|GetActorLocation (CallFunction|GetAvatarasFortPlayerPawn)) (Transformation|GetActorLocation (Variables|Default|GetLockActor))) -1.0)) (Transformation|GetActorLocation (Variables|Default|GetLockActor))))
      (:else))))

(event Custom|ReplicatedDrawTrailToVault
  (params (Rotation :type Rotator) (LockActorPos :type Vector) (ShouldBlockDraw :type Boolean))
  ;[1 1 1] From the server do the math and pass the rotation to the client. This is done so that if the actor is not relevant to the client the server still points in the correct direction.
  (if (not ShouldBlockDraw)
    (bind _returnvalue (Niagara|SpawnSystemAttached "/Game/Athena/Items/EnvironmentalItems/Locks/Keycard/FX/NS_Keycard_LockDirection.NS_Keycard_LockDirection" (Variables|/Script/Engine.Character|GetMesh (CallFunction|GetAvatarasFortPlayerPawn)) "root" "0.000000,0.000000,125.000000" "0, 0, 0" "SnapToTarget" true))
    (Effects|Components|ParticleSystem|SetVectorParameter _returnvalue "LockActorPos" LockActorPos)))

(fn InitAbility
  (bind _asfort_player_controller_athena (Utilities|Casting|CastToFortPlayerControllerAthena (Fortnite|GetFortPlayerControllerfromActor (CallFunction|GetAvatarasFortPlayerPawn))))
  (Variables|Default|SetPlayerController _asfort_player_controller_athena))

(fn PingMarkTarget
  (bind (_playerbankid _bankinstanceid) (Fortnite|RegisterSoundIndicatorPlacementContext (Variables|Default|GetPlayerController) (Variables|Default|GetLockActor) "0, 0, 0" false))
  (Variables|SetMarker_PlayerID (Utilities|Struct|MakeMarkerID _playerbankid _bankinstanceid)))

(fn RemoveCurrentMark
  (bind _asfort_player_controller_athena (Utilities|Casting|CastToFortPlayerControllerAthena (Ability|GetPlayerControllerfromActorInfo)))
  (MarkerComponent|UnmarkActoronClient (Class|FortPlayerController|GetMarkerComponent _asfort_player_controller_athena) (Variables|Default|GetMarker)))
